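; Blueprint function library of shared helpers for Fortnite AI pawns and bot controllers.
; Helpers edit loose gameplay tags, weapons, ammo, goal actors and radial leashes, and query angles, damageability and converted pawns.
; The blackboard target lookup feeds noise events; speech bubbles and VO use the Event.NPC.TriggerVO gameplay event.
; NPC cleanup tags or force-kills a nearby pawn without loot via an InstantDeath tag, returning from inside its loop after the first.
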(fn AddUniqueGameplayTag
  (params (Actor :type Actor%20Object%20Reference) (GameplayTag :type Gameplay%20Tag%20Structure) (Replicate :type Boolean))
  (bind _returnvalue (Ability|GetAbilitySystemComponent Actor))
  (Utilities|IsValid _returnvalue
    (:Is%20Valid
      (if (== (GameplayTags|GetGameplayTagCount _returnvalue GameplayTag) 0)
        (bind _returnvalue_2 (Ability|GameplayEffect|AddLooseGameplayTags Actor (GameplayTags|MakeGameplayTagContainerfromTag GameplayTag) (Variables|GetReplicate)))
        (return _returnvalue_2)
        (else
          _)))
    (:Is%20Not%20Valid
      _))
  (return false))

(fn CreateNoiseEventAtAITargetLocation
  (params (AIPawn :type Fort%20Pawn%20Object%20Reference) (MaxRange :type Float%20%28double-precision%29) (Noise :type Name))
  (bind _targetactor (CallFunction|GetTargetActorfromBlackboard AIPawn))
  (Utilities|IsValid _targetactor
    (:Is%20Valid
      (AI|Perception|MakeNoiseEventatLocation _targetactor MaxRange (Transformation|GetActorLocation _targetactor) Noise)
      (return true))
    (:Is%20Not%20Valid
      (return false))))

(fn ForceBestWeaponByTag
  (params (Pawn :type Fort%20Pawn%20Object%20Reference) (WeaponTag :type Gameplay%20Tag%20Structure))
  (Utilities|IsValid Pawn
    (:Is%20Valid
      (bind _asfort_athena_aibot_controller (Utilities|Casting|CastToFortAthenaAIBotController (Pawn|GetController Pawn))
        (:then
          (bind _output_get (LocalVariables|SetAIController _asfort_athena_aibot_controller))
          (bind _items (InventoryManagement|GetItemsfromInventoryOwner _output_get))
          (LocalVariables|SetFoundWeapon false)
          (bind _array_element (Utilities|Array|ForEachLoopwithBreak _items)
            (:LoopBody
              (bind _asfort_weapon_ranged_item_definition (Utilities|Casting|CastToFortWeaponRangedItemDefinition (Item|GetItemDefinition _array_element)))
              (if (GameplayTags|HasMatchingGameplayTag _asfort_weapon_ranged_item_definition WeaponTag)
                (Class|FortAthenaAIBotController|SetForcedBestWeapon (LocalVariables|GetAIController) _array_element)
                (LocalVariables|SetFoundWeapon true)))
            (:Completed
              _))
          (return (LocalVariables|GetFoundWeapon)))
        (:CastFailed
          _)))
    (:Is%20Not%20Valid
      _))
  (return false))

(fn GetAngleDegreesToTarget
  (params (Pawn :type Fort%20Pawn%20Object%20Reference) (TargetLocation :type Vector))
  (return (Class|AbilityFunctions|GetAngleBetweenTwoVectors (Transformation|GetActorForwardVector Pawn) (Math|Vector|Normalize (- TargetLocation (Transformation|GetActorLocation Pawn))) Pawn)))

(fn GetCanBeDamaged
  (params (Actor :type Actor%20Object%20Reference))
  (bind _asbuilding_actor (Utilities|Casting|CastToBuildingActor Actor)
    (:then
      (return true (Variables|/Script/Engine.Actor|GetbCanBeDamaged _asbuilding_actor)))
    (:CastFailed
      (bind _asfort_static_mesh_actor (Utilities|Casting|CastToFortStaticMeshActor Actor)
        (:then
          (return true (Variables|/Script/Engine.Actor|GetbCanBeDamaged _asfort_static_mesh_actor)))
        (:CastFailed
          (return false false))))))

(fn GetTargetActorFromBlackboard
  (params (AIPawn :type Fort%20Pawn%20Object%20Reference) (BlackBoardKeyName :type Name :default "AIEvaluator_Global_TargetActor"))
  (bind _returnvalue (Pawn|GetController AIPawn))
  (bind _returnvalue_2 (AI|GetBlackboard _returnvalue))
  (bind _returnvalue_3 (AI|Components|Blackboard|GetValueasObject _returnvalue_2 BlackBoardKeyName))
  (Utilities|IsValid AIPawn
    (:Is%20Valid
      (Utilities|IsValid _returnvalue
        (:Is%20Valid
          (Utilities|IsValid _returnvalue_2
            (:Is%20Valid
              (Utilities|IsValid _returnvalue_3
                (:Is%20Valid
                  (bind _asfort_pawn (Utilities|Casting|CastToFortPawn _returnvalue_3)
                    (:then
                      (return _asfort_pawn))
                    (:CastFailed
                      _)))
                (:Is%20Not%20Valid
                  _))
              (return 0))
            (:Is%20Not%20Valid
              _)))
        (:Is%20Not%20Valid
          _)))
    (:Is%20Not%20Valid
      _)))

(fn HasConvertedPawn
  (params (Actor :type Actor%20Object%20Reference))
  (bind _returnvalue (Actor|GetComponentbyClass Actor "/Script/FortniteGame.FortPawnComponent_Convert"))
  (Utilities|IsValid _returnvalue
    (:Is%20Valid
      (return (> (Convert|GetNumConvertedPawns _returnvalue "All") 0) (Convert|GetNumConvertedPawns _returnvalue "All")))
    (:Is%20Not%20Valid
      (return false 0))))

(fn PlaySpeechBubble
  (params (Pawn :type Fort%20Pawn%20Object%20Reference) (TagForSpeech :type Gameplay%20Tag%20Structure) (StopPreviousLine :type Boolean :default true) (TryPlayVOLine :type Boolean))
  (bind _returnvalue (Actor|GetComponentbyClass Pawn "/Script/FortniteGame.FortPawnComponent_SpeechBubble"))
  (Utilities|IsValid Pawn
    (:Is%20Valid
      (Utilities|IsValid _returnvalue
        (:Is%20Valid
          (LocalVariables|SetAlreadySpeaking (SpeechBubble|IsSpeaking _returnvalue))
          (if StopPreviousLine
            (SpeechBubble|StopSpeech _returnvalue)
            (else
              _))
          (SpeechBubble|StartSpeech _returnvalue (GameplayTags|MakeGameplayTagQueryMatchAnyTags (GameplayTags|MakeGameplayTagContainerfromTag TagForSpeech)))
          (if (and (Variables|GetTryPlayVOLine) (or (not (LocalVariables|GetAlreadySpeaking)) (Variables|GetStopPreviousLine)))
            (Ability|SendGameplayEventtoActor Pawn "(TagName=\"Event.NPC.TriggerVO\")" (Utilities|Struct|MakeGameplayEventData "(TagName=\"\")" 0 0 0 0 "()" "(GameplayTags=)" (GameplayTags|MakeGameplayTagContainerfromTag (Variables|GetTagForSpeech))))
            (return)
            (else
              (return))))
        (:Is%20Not%20Valid
          (Fortnite|FortShippingLog 0 "Trying to call Play Speech from AI_CommonFunctions but the pawn provided does not have a SpeechBubbleComponent." true)
          (return))))
    (:Is%20Not%20Valid
      (Fortnite|FortShippingLog 0 "Trying to call Play Speech from AI_CommonFunctions but did not provide a Pawn." true)
      (return))))

(fn ReloadPawnWeapon
  (params (PlayerPawn :type Fort%20Player%20Pawn%20Object%20Reference) (AmmoCount :type Integer :default 500))
  (bind _playerpawn (Variables|GetPlayerPawn))
  (bind _returnvalue (Actor|GetComponentbyClass _playerpawn "/Script/FortniteGame.FortOffhandComponent"))
  (bind _currentweapon (Variables|/Script/FortniteGame.FortPawn|GetCurrentWeapon PlayerPawn))
  (Utilities|IsValid _currentweapon
    (:Is%20Valid
      (FortWeapon|ModAmmo _currentweapon AmmoCount)
      (Utilities|IsValid _returnvalue
        (:Is%20Valid
          (bind _returnvalue_2 (Class|FortOffhandComponent|GetCurrentEquippedOffhandWeapon _returnvalue))
          (Utilities|IsValid _returnvalue_2
            (:Is%20Valid
              (FortWeapon|ModAmmo _returnvalue_2 (Variables|GetAmmoCount))
              (return))
            (:Is%20Not%20Valid
              _)))
        (:Is%20Not%20Valid
          _))
      (return))
    (:Is%20Not%20Valid
      _)))

(fn RemoveAllInstancesOfGameplayTag
  (params (Actor :type Actor%20Object%20Reference) (GameplayTag :type Gameplay%20Tag%20Structure) (Replicate :type Boolean))
  (bind _returnvalue (Ability|GetAbilitySystemComponent Actor))
  (bind _returnvalue_2 (GameplayTags|GetGameplayTagCount _returnvalue GameplayTag))
  (Utilities|IsValid _returnvalue
    (:Is%20Valid
      (bind _output_get (LocalVariables|SetTagCount _returnvalue_2))
      (if (> _returnvalue_2 0)
        (for _ (range _output_get)
          (bind _returnvalue_3 (Ability|GameplayEffect|RemoveLooseGameplayTags Actor (GameplayTags|MakeGameplayTagContainerfromTag GameplayTag) (Variables|GetReplicate))))
        (return (stale _returnvalue_3))
        (else
          _)))
    (:Is%20Not%20Valid
      _))
  (return false))

(fn SetGoalActor
  (params (AIPawn :type Actor%20Object%20Reference) (NewGoalActor :type Actor%20Object%20Reference))
  (bind _aspawn (Utilities|Casting|CastToPawn AIPawn)
    (:then
      (bind _returnvalue (AI|GoalSelection|GetAIGoalComponent (Pawn|GetController _aspawn)))
      (Utilities|IsValid _returnvalue
        (:Is%20Valid
          (AI|SetGoalActor _returnvalue NewGoalActor false)
          (return true))
        (:Is%20Not%20Valid
          _)))
    (:CastFailed
      _))
  (return true))

(fn SetRadialLeash
  (params (Target :type Fort%20Athena%20AIBot%20Controller%20Object%20Reference) (Location :type Vector%20%28by%20ref%29) (InnerRadius :type Float%20%28single-precision%29) (OuterRadius :type Float%20%28single-precision%29) (BypassCombatLeashOverride :type Boolean :default true))
  (Utilities|IsValid Target
    (:Is%20Valid
      (LocalVariables|SetBotController (Variables|GetTarget))
      (Class|FortAthenaAIBotController|SetRadialLeashInnerRadius (LocalVariables|GetBotController) (Variables|GetInnerRadius))
      (Class|FortAthenaAIBotController|SetRadialLeashOuterRadius (LocalVariables|GetBotController) (Variables|GetOuterRadius))
      (Class|FortAthenaAIBotController|SetRadialLeashLocation (LocalVariables|GetBotController) (Variables|GetLocation))
      (if (Variables|GetBypassCombatLeashOverride)
        (Class|FortAthenaAIBotController|SetRadialLeashAlertLevelOverride (LocalVariables|GetBotController) (Utilities|Struct|MakeLeashInfoOverride (Variables|GetInnerRadius) (Variables|GetOuterRadius)) "Suspicious")
        (Class|FortAthenaAIBotController|SetRadialLeashAlertLevelOverride (LocalVariables|GetBotController) (Utilities|Struct|MakeLeashInfoOverride (Variables|GetInnerRadius) (Variables|GetOuterRadius)) "LostTarget")
        (Class|FortAthenaAIBotController|SetRadialLeashAlertLevelOverride (LocalVariables|GetBotController) (Utilities|Struct|MakeLeashInfoOverride (Variables|GetInnerRadius) (Variables|GetOuterRadius)) "Threatened")
        (return)
        (else
          (return))))
    (:Is%20Not%20Valid
      (return))))

(fn StartLocationNpcCleanup
  (params (LocationActor :type Actor%20Object%20Reference) (ActorClassToCheck :type Actor%20Class%20Reference) (CleanupGameplayTag :type Gameplay%20Tag%20Container%20Structure) (DistanceToCheck :type Float%20%28double-precision%29 :default 1.25e+04))
  (LocalVariables|SetIsTagNotEmpty (Utilities|GameplayTagContainerNotEmpty (Variables|GetCleanupGameplayTag)))
  (bind _outactors (Actor|GetAllActorsOfClass 0 (Variables|GetActorClassToCheck)))
  (for _array_element _outactors
    (Utilities|IsValid (Variables|GetLocationActor)
      (:Is%20Valid
        (if (<= (Transformation|GetDistanceTo (Variables|GetLocationActor) _array_element) (Variables|GetDistanceToCheck))
          (bind _asfort_player_pawn_athena (Utilities|Casting|CastToFortPlayerPawnAthena _array_element)
            (:then
              (LocalVariables|SetBotPawn _asfort_player_pawn_athena)
              (if (LocalVariables|GetIsTagNotEmpty)
                (Ability|GameplayEffect|AddLooseGameplayTags (LocalVariables|GetBotPawn) (Variables|GetCleanupGameplayTag))
                (bind _output_get (LocalVariables|SetSucceed true))
                (return _output_get)
                (else
                  ;[1 1 1] InstantDeath tag is to prevent loot from being drop when eliminated. This tag must be managed in the AILootOnDeathDataTable as an optional tag query with a drop or not drop loot entry. This tag is pushed on the pawn for it to work.
                  (Ability|GameplayEffect|AddLooseGameplayTags (LocalVariables|GetBotPawn) (LocalVariables|GetInstantDeathTag))
                  (FortPawn|ForceKill (LocalVariables|GetBotPawn) (LocalVariables|GetDeathReasonTag) 0 0)
                  (bind _output_get_2 (LocalVariables|SetSucceed true))
                  (return _output_get_2))))
            (:CastFailed
              (return (LocalVariables|GetSucceed))))))
      (:Is%20Not%20Valid
        (return false)))))

(fn TriggerVOFromServer
  (params (Pawn :type Fort%20Pawn%20Object%20Reference) (VOTag :type Gameplay%20Tag%20Structure))
  (if (GameplayTags|IsGameplayTagValid VOTag)
    (Utilities|IsValid Pawn
      (:Is%20Valid
        (Ability|SendGameplayEventtoActor Pawn "(TagName=\"Event.NPC.TriggerVO\")" (Utilities|Struct|MakeGameplayEventData "(TagName=\"\")" 0 0 0 0 "()" "(GameplayTags=)" (GameplayTags|MakeGameplayTagContainerfromTag VOTag))))
      (:Is%20Not%20Valid
        _))
    (else
      _))
  (return))
